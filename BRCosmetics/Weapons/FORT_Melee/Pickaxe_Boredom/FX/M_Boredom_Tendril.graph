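# Unlit, two-sided translucent particle material for the Boredom pickaxe tendril effect, usable on Niagara sprites and mesh particles.
# A rotated plant texture's red channel, minus the dynamic Erosion parameter and a debug offset, is hard-cut with Ceil and multiplied by particle alpha as an opacity mask.
# A second cut, inset by Toon Line Thickness, masks particle color into an emissive adjustment function, leaving a dark toon outline.
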Material /BRCosmetics/Weapons/FORT_Melee/Pickaxe_Boredom/FX/M_Boredom_Tendril
graph {
 properties {
  bUsedWithNiagaraMeshParticles=true
  bUsedWithNiagaraSprites=true
  blendMode=BLEND_Translucent
  pixelDepthOffsetMode=PDOM_Legacy
  shadingModel=MSM_Unlit
  twoSided=true
  usedShadingModels=Unlit
 }
 nodes [
  n1 = ParticleColor()
  n2 = ScalarParameter()
  n3 = TextureCoordinate()
  n4 = Subtract(
   A=TextureSample(UVs=MaterialFunctionCall!(Angle=@n2, UV/2DVector=@n3) {materialFunction=/Game/Characters/Player/Male/Medium/Bodies/M_MED_Faux/Materials/MF_2DRotate}.Result) {automaticViewMipBias=false, texture=&/T_FX_Boredom_Plant}.R
   B=Add(
    A=ScalarParameter() {parameterName=DebugErosion}
    B=DynamicParameter() {defaultValue=<rgba 0 0 0 0>, paramNames [Erosion,Param2,Param3,Param4]}.Erosion
   )
  ) {constB=0.1}
  n5 = Add(A=@n4, B=Multiply(A=ScalarParameter() {defaultValue=0.25, parameterName=Toon Line Thickness}) {constB=-1})
  MaterialFunctionCall(Emissive=Multiply(A=@n1.RGB, B=Ceil(None=@n5))) {materialFunction=&MF_FX_EmissiveAdjustments}
  Multiply(A=@n1.A, B=Ceil(None=@n4))
 ]
}
